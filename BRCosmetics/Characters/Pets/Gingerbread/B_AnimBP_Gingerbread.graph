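# B_AnimBP_Gingerbread is an Animation Blueprint for the Gingerbread pet cosmetic, targeting the Chameleon pet skeleton.
# It is an empty shell here: the event graph has no nodes and the Blueprint declares no variables.
# Its use of Chameleon_Skeleton shows the Gingerbread pet shares that skeleton rather than one of its own.
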AnimBlueprint /BRCosmetics/Characters/Pets/Gingerbread/B_AnimBP_Gingerbread
target_skeleton=/Game/Characters/Pets/Chameleon/Meshes/Chameleon_Skeleton
graphs [EventGraph {links [], nodes []}]
variables []
