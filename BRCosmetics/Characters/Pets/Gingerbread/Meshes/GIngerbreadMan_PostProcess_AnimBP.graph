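# Post-process AnimBlueprint for the Gingerbread pet, targeting the Chameleon pet skeleton, that adds secondary physics motion to the neck and head.
# The cached input pose feeds two chained Anim Dynamics nodes on neck_02 and head, using damped angular springs with tight limits and an LOD threshold.
# bPetWantsAnimDynamics blends linearly between the simulated pose and the untouched input pose.
# The event graph holds only unconnected stub nodes.
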AnimBlueprint /BRCosmetics/Characters/Pets/Gingerbread/Meshes/GIngerbreadMan_PostProcess_AnimBP
target_skeleton=/Game/Characters/Pets/Chameleon/Meshes/Chameleon_Skeleton
graphs [
 AnimGraph {nodes [
   n1 = Anim Dynamics - Bone: neck_02(ComponentPose=Local To Component(LocalPose=Use cached pose 'Sub Graph Cache'()), Alpha=1, ExternalForce="0,0,0") {
    angularDampingOverride=0.999
    angularSpringConstant=200
    bGravityOverrideInSimSpace=true
    bOverrideAngularDamping=true
    lODThreshold=1
    numSolverIterationsPostUpdate=4
    numSolverIterationsPreUpdate=16
    alphaBoolBlend {blendOption=Linear}
    alphaScaleBiasClamp {inRange {max=500}}
    boundBone {boneName=neck_02}
    componentAppliedLinearAccClamp=<500 500 500>
    componentLinearAccScale=<0.5 0.5 0.5>
    physicsBodyDefinitions [{
      sphereCollisionRadius=0
      boundBone {boneName=neck_02}
      boxExtents=<10 3 10>
      constraintSetup {angularLimitsMax=<6 10 3>, angularLimitsMin=<-6 -10 -3>, angularTarget=<0 0 0>}
      localJointOffset=<1 0 0>
     }]
   }
   n2 = Anim Dynamics - Bone: head(ComponentPose=@n1, Alpha=1, ExternalForce="0,0,0") {
    angularDampingOverride=0.999
    angularSpringConstant=200
    bGravityOverrideInSimSpace=true
    bOverrideAngularDamping=true
    lODThreshold=1
    numSolverIterationsPostUpdate=4
    numSolverIterationsPreUpdate=16
    alphaBoolBlend {blendOption=Linear}
    alphaScaleBiasClamp {inRange {max=500}}
    boundBone {boneName=head}
    componentAppliedLinearAccClamp=<500 500 500>
    componentLinearAccScale=<0.5 0.5 0.5>
    physicsBodyDefinitions [{
      sphereCollisionRadius=0
      boundBone {boneName=head}
      boxExtents=<3 10 10>
      constraintSetup {angularTargetAxis=AxisY, twistAxis=AxisY, angularLimitsMax=<10 3 3>, angularLimitsMin=<-10 -3 -3>, angularTarget=<0 0 0>}
      localJointOffset=<0 1 0>
     }]
   }
   Output Pose(Result=Blend Poses by bool(
     BlendPose_0=Component To Local(ComponentPose=@n2)
     BlendPose_1=Use cached pose 'Sub Graph Cache'()
     BlendTime_0=1.2
     BlendTime_1=1.2
     bActiveValue=Get bPetWantsAnimDynamics() -> (bPetWantsAnimDynamics:bool)
    ) {blendType=Linear}) {layerGroup=None}
   AnimGraphNode_SubInput_0 = Input Pose()
   Save cached pose 'Sub Graph Cache'(Pose=@AnimGraphNode_SubInput_0) {cacheName=Sub Graph Cache}
  ]}
 EventGraph {nodes [
   TryGetPawnOwner()
   Event BlueprintUpdateAnimation()
  ]}
]
variables []
